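# Material function that restyles incoming material attributes into the fuzzy, metallic gold look of the Golden Fleece statuette quest prop.
# It overrides base color, normal, metallic and roughness, using velvet textures, fuzzy shading, a hue shift and Fresnel-driven metallic.
# Named reroutes carry shared terms, including an eyes mask, camera-distance factor, and sparkles built from animated glitter plus six sine-timed sphere-mask bursts.
# The pulsing star emissive chain and the attribute blend are left unconnected.
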MaterialFunction /Game/Environments/Helios/Props/Quest/Quest_GoldenFleeceStatuette_A/Materials/MF_GoldenFleeceStatuette
graph {nodes [
  n1 = Add(A=NamedRerouteUsage()) {constB=0.15}
  n2 = NamedRerouteUsage()
  n3 = Add(A=@n2) {constB=-0.2}
  n4 = Add(A=@n2) {constB=0.1}
  n5 = SetMaterialAttributes!(
   MaterialAttributes=NamedRerouteUsage()
   EmissiveColor
   Normal=LinearInterpolate(A=NamedRerouteUsage(), B=NamedRerouteUsage(), Alpha=Saturate(None=@n1))
   BaseColor=NamedRerouteUsage()
   Metallic=NamedRerouteUsage()
   Roughness=LinearInterpolate(A=Saturate(None=@n3), B=Saturate(None=@n4), Alpha=NamedRerouteUsage())
  ) {attributeSetTypes [
    #1
    #2
    #3
    #4
    #5
   ]}
  BlendMaterialAttributes(A=NamedRerouteUsage(), B=@n5, Alpha=NamedRerouteUsage())
  Constant() {r=1}
  n6 = TextureSample(UVs=Multiply(A=TextureCoordinate()) {constB=2}) {
   samplerSource=SSM_Wrap_WorldGroupSettings
   samplerType=SAMPLERTYPE_Masks
   texture=/Game/Environments/Helios/Props/Quest/Quest_GoldenFleeceStatuette_A/Textures/T_Velvet_GrungeMask1
  }
  n7 = Saturate(None=Add(A=@n6.R) {constB=0.2})
  n8 = TextureCoordinate()
  n9 = Multiply(A=Time()) {constB=0.723}
  n10 = Floor(None=@n9)
  n11 = SmoothStep(Value=Frac(None=@n9)) {constMax=0.95, constMin=0.05}
  n12 = TextureSample(UVs=Panner(Coordinate=Multiply(A=@n8) {constB=2}, Time=Add(A=@n10, B=@n11)) {speedX=-0.54, speedY=0.1}) {
   samplerSource=SSM_Wrap_WorldGroupSettings
   samplerType=SAMPLERTYPE_Masks
   texture=&T_Velvet_GrungeMask1
  }
  n13 = TextureCoordinate()
  n14 = Constant() {r=11}
  n15 = Sine(None=Multiply(A=Time()) {constB=15})
  n16 = TextureCoordinate()
  n17 = TextureSample!(
   UVs=Add(A=Multiply(A=@n13, B=@n14) {constB=5.5}, B=Multiply(A=@n15) {constB=0.0125})
   Tex
   "DDX(UVs)"=DDX(Value=@n16)
   "DDY(UVs)"=DDY(Value=@n16)
   Apply View MipBias
  ) {
   mipValueMode=TMVM_Derivative
   samplerSource=SSM_Wrap_WorldGroupSettings
   samplerType=SAMPLERTYPE_Masks
   texture=/Game/Environments/Helios/Props/Quest/Quest_GoldenFleeceStatuette_A/Textures/T_Quest_GoldenFleeceStatuette_ShinyStars
  }
  n18 = Add(B=@n17.G)
  n19 = Add(A=@n17.B, B=Multiply(A=Time()) {constB=0.42})
  n20 = Add(A=Multiply(A=Sine(None=@n19)) {constB=0.5}) {constB=0.5}
  n21 = Constant() {r=60}
  n22 = Multiply(
   A=Multiply(A=Multiply(A=@n18, B=@n20), B=Constant3Vector() {constant=<rgba 1 0.6757 0.0521 1>})
   B=MaterialFunctionCall(Value_Morning=@n21, Value_Day=@n21, Value_Evening=@n21, Value_Night=Constant() {r=29}) {materialFunction=&Scalar_Time_of_Day}.Result
  ) {constB=45}
  n23 = NamedRerouteUsage()
  Divide(A=Multiply(A=Multiply(A=@n7, B=@n12.R), B=Multiply(A=@n22, B=@n23)), B=EyeAdaptation().EyeAdaptation)
  FunctionOutput(None=@n5)
  n24 = TextureObject() {
   samplerType=SAMPLERTYPE_Masks
   texture=&T_Quest_GoldenFleeceStatuette_ShinyStars
  }
  MaterialFunctionCall(
   TextureObject=@n24
   WorldPosition=Multiply(A=NamedRerouteUsage(), B=ScalarParameter() {defaultValue=1.5, group=GoldenFleece, parameterName=Stars_Tiling, sortPriority=11}) {constB=1.5}
   World Space Normal=VertexInterpolator(VS=PreSkinnedNormal()).PS
  ) {materialFunction=/Game/Weapons/WeaponSkins/Wraps/MaterialFunctions/WorldAlignedTexture}
  Multiply(A=NamedRerouteUsage()) {constB=0.2}
  n25 = TextureCoordinate()
  n26 = TextureSample!(UVs=NamedRerouteDeclaration(None=Multiply(A=@n25) {constB=1.5}) {name=UVFuzz, nodeColor=<rgba 0 0.2266 1 1>}, Tex, Apply View MipBias) {
   samplerSource=SSM_Wrap_WorldGroupSettings
   samplerType=SAMPLERTYPE_Normal
   texture=/Game/Environments/Helios/Props/Quest/Quest_GoldenFleeceStatuette_A/Textures/T_Velvet_N1
  }
  n27 = Constant() {r=0.5}
  NamedRerouteDeclaration(None=MaterialFunctionCall(
    BaseNormal=NamedRerouteUsage()
    AdditionalNormal=MaterialFunctionCall(Normal=@n26.RGB, Flatness=@n27) {materialFunction=&FlattenNormal}.Result
   ) {materialFunction=&BlendAngleCorrectedNormals}.Result) {name=Normal, nodeColor=<rgba 1 0 0.9999 1>}
  n28 = GetMaterialAttributes(None=FunctionInput() {inputName=MaterialAttributes, inputType=FunctionInput_MaterialAttributes}) {attributeGetTypes [
    #1
    #2
    #3
    #4
    #5
   ]}
  NamedRerouteDeclaration(None=@n28.Normal) {name=OriginalNormal, nodeColor=<rgba 1 0.6094 0 1>}
  NamedRerouteDeclaration(None=@n28.Roughness) {name=OriginalRoughness, nodeColor=<rgba 1 0 0.3984 1>}
  n29 = SphereMask(A=NamedRerouteUsage(), B=Constant3Vector() {constant=<rgba 5 -10 12 0>}) {attenuationRadius=28, hardnessPercent=0}
  n30 = Multiply(A=@n29, B=@n29)
  n31 = ScalarParameter() {defaultValue=0.25, group=GoldenFleece, parameterName=Bursts_Speed}
  n32 = Multiply(A=Add(A=Time()) {constB=0.1}, B=@n31)
  n33 = Saturate(None=Sine(None=Add(A=@n30, B=@n32)))
  n34 = SphereMask(A=NamedRerouteUsage(), B=Constant3Vector() {constant=<rgba -6 -10 15 0>}) {attenuationRadius=23, hardnessPercent=0}
  n35 = Multiply(A=@n34, B=@n34)
  n36 = Time()
  n37 = Add(A=@n35, B=Multiply(A=Add(A=@n36) {constB=0.82}, B=@n31))
  n38 = Multiply(A=@n35, B=Saturate(None=Sine(None=@n37)))
  n39 = SphereMask(A=NamedRerouteUsage(), B=Constant3Vector() {constant=<rgba -17 -1.75 45 0>}) {attenuationRadius=23, hardnessPercent=0}
  n40 = Multiply(A=@n39, B=@n39)
  n41 = Add(A=Time()) {constB=0.3}
  n42 = Sine(None=Add(A=@n40, B=Multiply(A=@n41, B=@n31)))
  n43 = SphereMask(A=NamedRerouteUsage(), B=Constant3Vector() {constant=<rgba -13 5 20 0>}) {attenuationRadius=30, hardnessPercent=0}
  n44 = Multiply(A=@n43, B=@n43)
  n45 = Multiply(A=Add(A=Time()) {constB=0.55}, B=@n31)
  n46 = Saturate(None=Sine(None=Add(A=@n44, B=@n45)))
  n47 = SphereMask(A=NamedRerouteUsage(), B=Constant3Vector() {constant=<rgba 18 -8 20 0>}) {attenuationRadius=20, hardnessPercent=0}
  n48 = Multiply(A=@n47, B=@n47)
  n49 = Time()
  n50 = Add(A=@n48, B=Multiply(A=Add(A=@n49) {constB=0.37}, B=@n31))
  n51 = Multiply(A=@n48, B=Saturate(None=Sine(None=@n50)))
  n52 = SphereMask(A=NamedRerouteUsage(), B=Constant3Vector() {constant=<rgba 0 8 18 0>}) {attenuationRadius=34, hardnessPercent=0}
  n53 = Multiply(A=@n52, B=@n52)
  n54 = Add(A=@n53, B=Multiply(A=Time(), B=@n31))
  n55 = Multiply(A=@n53, B=Saturate(None=Sine(None=@n54)))
  n56 = Add(A=Multiply(A=@n40, B=Saturate(None=@n42)), B=Add(A=Multiply(A=@n44, B=@n46), B=Add(A=@n51, B=@n55)))
  n57 = Multiply(A=Add(A=Multiply(A=@n30, B=@n33), B=Add(A=@n38, B=@n56)), B=Constant3Vector() {constant=<rgba 1 0.6109 0.1818 1>})
  n58 = ScalarParameter() {defaultValue=0.2, group=GoldenFleece, parameterName=Bursts_Intensity, sortPriority=11}
  n59 = Multiply(
   A=Multiply(
    A=MaterialFunctionCall() {materialFunction=&/MF_GlitterFX_GoldenFleeceStatuette}.GlitterAnimated
    B=Constant3Vector() {constant=<rgba 0.2865 0.158 0.0164 1>}
   )
   B=ScalarParameter() {defaultValue=3, group=GoldenFleece, parameterName=Sparkles_Intensity, sortPriority=11}
  ) {constB=3}
  n60 = NamedRerouteUsage()
  NamedRerouteDeclaration(None=Add(A=Multiply(A=@n57, B=@n58) {constB=0.2}, B=Multiply(A=@n59, B=@n60))) {name=Sparkles, nodeColor=<rgba 1 0.1875 0 1>}
  n61 = OneMinus(None=SphereMask(A=NamedRerouteUsage(), B=Constant()) {attenuationRadius=0.3, hardnessPercent=42})
  NamedRerouteDeclaration(None=@n61) {name=EyesMask, nodeColor=<rgba 0 1 0.5313 1>}
  n62 = Distance(A=ObjectPositionWS(), B=CameraPositionWS())
  NamedRerouteDeclaration(None=Saturate(None=InverseLinearInterpolate(Value=@n62) {constA=180, constB=450})) {name=DistanceBlending, nodeColor=<rgba 0.3057 0.2258 1 1>}
  NamedRerouteDeclaration(None=VertexInterpolator(VS=PreSkinnedPosition()).PS) {name=PreskinnedPosition, nodeColor=<rgba 1 0 0.3281 1>}
  n63 = Multiply(A=Desaturation(None=NamedRerouteUsage(), Fraction=Constant() {r=1})) {constB=8}
  n64 = TextureSample(UVs=NamedRerouteUsage()) {
   samplerSource=SSM_Wrap_WorldGroupSettings
   samplerType=SAMPLERTYPE_Masks
   texture=/Game/Environments/Helios/Props/Quest/Quest_GoldenFleeceStatuette_A/Textures/T_Velvet_Height1
  }
  n65 = Saturate(None=Add(A=@n64.RGB) {constB=0.15})
  n66 = Multiply(A=NamedRerouteUsage()) {constB=3.5}
  n67 = Multiply(A=@n6.R, B=Constant3Vector() {constant=<rgba 1 0.6696 0.1583 1>})
  n68 = LinearInterpolate(
   A=LinearInterpolate(
    A=Constant3Vector() {constant=<rgba 0.5365 0.3986 0.1118 1>}
    B=Constant3Vector() {constant=<rgba 1.75 1.2915 0.4194 1>}
    Alpha=Saturate(None=@n63)
   )
   B=Add(A=Multiply(A=@n65, B=@n66), B=Multiply(A=@n67) {constB=0.4})
   Alpha=NamedRerouteUsage()
  )
  n69 = NamedRerouteUsage()
  n70 = Constant() {r=0.7}
  n71 = Constant() {r=1}
  n72 = Constant() {r=0.2}
  n73 = MaterialFunctionCall(
   Hue Shift Percentage=Constant() {r=-0.02}
   Texture=Desaturation(
    None=MaterialFunctionCall!(BaseColor=@n68, Normal=@n69, CoreDarkness=@n70, Power=@n71, EdgeBrightness=@n72) {materialFunction=/Engine/Functions/Engine_MaterialFunctions01/Shading/FuzzyShading}.Result
    Fraction=Constant() {r=-0.25}
   )
  ) {materialFunction=&HueShift}
  n74 = NamedRerouteUsage()
  NamedRerouteDeclaration(None=Multiply(A=Add(A=@n73.Result, B=@n74), B=NamedRerouteUsage())) {name=BaseColor, nodeColor=<rgba 1 0.8672 0 1>}
  NamedRerouteDeclaration(None=@n28.BaseColor) {name=OriginalBaseColor, nodeColor=<rgba 1 0.1406 0 1>}
  NamedRerouteDeclaration(None=@n28.MaterialAttributes) {name=OriginalMA, nodeColor=<rgba 1 0 0.1015 1>}
  NamedRerouteDeclaration(None=@n28.Metallic) {name=OriginalMetallic, nodeColor=<rgba 0 0.8594 1 1>}
  n75 = Fresnel(Normal=NamedRerouteUsage()) {exponent=6}
  NamedRerouteDeclaration(None=Multiply(A=LinearInterpolate(Alpha=@n75) {constA=0.8, constB=0.6}, B=NamedRerouteUsage())) {name=Metallic, nodeColor=<rgba 0 0.2266 1 1>}
  Reroute(None=@n61)
 ]}
